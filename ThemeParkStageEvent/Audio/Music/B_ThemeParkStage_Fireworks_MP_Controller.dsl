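; B_ThemeParkStage_Fireworks_MP_Controller is a child Blueprint that controls music for the theme park stage fireworks event.
; On BeginPlay it runs the parent logic, then sends its audio component post-effect to an audio bus at full level.
; StartMusicParty is an empty event with no parent call, while StartFireWorkMusic and the construction script only forward to the parent, which holds the playback logic.
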(event EventBeginPlay
  (|CallParentFunction)
  (Audio|Components|Audio|SetAudioBusSendPostEffect (Variables|Default|GetAudio) (Variables|Audio|GetAudioBus) 1.0 :!class /Script/Engine.AudioComponent))

(event EventStartMusicParty)

(event Custom|StartFireWorkMusic
  (|CallParentFunction))

(fn ConstructionScript
  (|CallParentFunction))
