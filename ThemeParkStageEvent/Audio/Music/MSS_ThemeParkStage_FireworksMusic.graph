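# Stereo one-shot MetaSound source that plays the fireworks music for the Theme Park Stage event.
# OnPlay starts a single stereo Wave Player, which feeds both outputs and fires OnFinished when done.
# Base submix and submix sends are disabled, so audio goes to the fireworks speakers audio bus instead of playing directly.
# It uses the base music sound class and keeps playing when silent.
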MetaSoundSource /ThemeParkStageEvent/Audio/Music/MSS_ThemeParkStage_FireworksMusic
inputs [UE.Source.OnPlay {type=Trigger}]
interfaces [UE.OutputFormat.Stereo:1.0, UE.Source.OneShot:1.0, UE.Source:1.1]
outputs [
 "UE.OutputFormat.Stereo.Audio:0" {type=Audio}
 "UE.OutputFormat.Stereo.Audio:1" {type=Audio}
 UE.Source.OneShot.OnFinished {type=Trigger}
]
pages [Default {
  input_defaults {UE.Source.OnPlay=null}
  nodes [WavePlayer_0 = UE.Wave Player:Stereo(Play=^UE.Source.OnPlay) {also {defaults {Wave Asset=&/MX_S39_WonkeeLand_Fireworks_LP}, version [1,1]}}]
  links [
   {dst="@UE.Source.OneShot.OnFinished", src=WavePlayer_0.On Finished}
   {dst="@UE.OutputFormat.Stereo.Audio:0", src=WavePlayer_0.Out Left}
   {dst="@UE.OutputFormat.Stereo.Audio:1", src=WavePlayer_0.Out Right}
  ]
 }]
properties {
 assetImportData=&self:AssetImportData
 bEnableBaseSubmix=false
 bEnableSubmixSends=false
 numChannels=2
 outputFormat=Stereo
 priority=100
 sampleRate=48000
 soundClassObject=/Game/Sounds/FortSoundClassesAndModes/Fort_11_Base_Music.Fort_11_Base_music
 virtualizationMode=PlayWhenSilent
 volume=3
 busSends [{audioBus=&/AB_ThemeParkStage_Speakers_Fireworks, sendLevel=1}]
}
